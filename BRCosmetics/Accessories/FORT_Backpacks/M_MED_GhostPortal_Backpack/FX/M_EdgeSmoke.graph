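# Translucent, two-sided unlit material for scrolling edge smoke on the Ghost Portal backpack cosmetic.
# Cloud noise distorts panned turbulence, which blends two colors and shapes opacity, with vertex red adding a hot band.
# Opacity fades by view angle and depth (both skipped on mobile), near the camera, and via BP_Opacity.
# BP_OffsetPos, weighted by a squared UV gradient and scrolling noise, supplies an offset, likely for vertex displacement.
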Material /BRCosmetics/Accessories/FORT_Backpacks/M_MED_GhostPortal_Backpack/FX/M_EdgeSmoke
graph {
 properties {blendMode=BLEND_Translucent, pixelDepthOffsetMode=PDOM_Legacy, shadingModel=MSM_Unlit, twoSided=true, usedShadingModels=Unlit}
 nodes [
  n1 = VectorParameter() {parameterName=LerpA, defaultValue=<rgba 0.0342 0.0832 0.224 1>}
  n2 = VectorParameter() {parameterName=LerpB, defaultValue=<rgba 0.0339 0.4896 0.4506 1>}
  n3 = Panner(Coordinate=TextureCoordinate() {coordinateIndex=1, uTiling=3, vTiling=0.5}) {speedY=-0.1}
  n4 = Panner(Coordinate=TextureCoordinate() {coordinateIndex=1, uTiling=2, vTiling=0.5}) {speedY=-0.25}
  n5 = ConstantBiasScale(None=ComponentMask(None=TextureSample(UVs=@n4) {samplerType=SAMPLERTYPE_LinearColor, texture=/Game/Missions/Primary/DeliverTheBomb/Textures/T_CloudNoise}.RGB) {g=true, r=true}) {bias=-0.5, scale=0.1}
  n6 = TextureSample(UVs=Add(A=Add(A=@n3) {constB=0.01}, B=@n5)) {texture=/Game/Global/Textures/Default/Caustics/T_Turbulence_Seamless}
  n7 = VectorParameter() {parameterName=ColorMutiply, defaultValue=<rgba 0.0221 0.3374 0.5 1>}
  n8 = VertexColor()
  n9 = Multiply(A=Saturate(None=Subtract(A=@n8.R) {constB=0.5})) {constB=2}
  n10 = Multiply(A=@n9, B=@n9)
  MaterialFunctionCall(Emissive=Add(A=LinearInterpolate(A=@n1.RGB, B=@n2.RGB, Alpha=@n6.R), B=Multiply(A=@n7.RGB, B=@n10))) {materialFunction=&MF_FX_EmissiveAdjustments}
  n11 = Multiply(
   A=LinearInterpolate(Alpha=Multiply(A=@n8.R, B=@n8.R)) {constB=2}
   B=TextureSample(UVs=Add(A=@n3, B=@n5)) {texture=&T_Turbulence_Seamless}.R
  )
  n12 = Abs(None=DotProduct(A=CameraVectorWS(), B=VertexNormalWS()))
  n13 = ShadingPathSwitch(Default=Multiply(A=Subtract(A=@n12) {constB=0.1}) {constB=2}, Mobile=Constant() {r=1}) {desc=Migrated from FeatureLevelSwitch}
  n14 = Multiply(
   A=Add(A=@n10, B=Multiply(A=@n11, B=@n13) {constB=2})
   B=ShadingPathSwitch(Default=DepthFade() {fadeDistanceDefault=1}, Mobile=Constant() {r=1}) {desc=Migrated from FeatureLevelSwitch}
  )
  n15 = WorldPosition()
  n16 = CameraPositionWS()
  n17 = Constant() {r=150}
  n18 = Constant() {r=0.5}
  Multiply(
   A=Multiply(A=Multiply(A=@n14) {constB=0.5}, B=ScalarParameter() {defaultValue=1, parameterName=BP_Opacity})
   B=OneMinus(None=SphereMask(A=@n15.XYZ, B=@n16, Radius=@n17, Hardness=@n18))
  )
  n19 = TextureCoordinate() {coordinateIndex=1, uTiling=1.4, vTiling=1.4}
  n20 = Saturate(None=ComponentMask(None=Subtract(A=@n19) {constB=0.4}) {g=true})
  n21 = Multiply(A=VectorParameter() {parameterName=BP_OffsetPos, defaultValue=<rgba 0 0 0 1>}.RGB, B=Multiply(A=@n20, B=@n20))
  n22 = Panner(Coordinate=TextureCoordinate() {coordinateIndex=1, vTiling=0.25}) {speedY=-0.25}
  Multiply(
   A=@n21
   B=LinearInterpolate(Alpha=TextureSample(UVs=@n22) {samplerType=SAMPLERTYPE_LinearColor, texture=&T_CloudNoise}.R) {constA=0.4, constB=0}
  )
  Multiply(A=@n21) {constB=0.2}
 ]
}
