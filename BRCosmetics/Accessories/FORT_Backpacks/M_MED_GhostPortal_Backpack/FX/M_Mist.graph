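# Translucent unlit particle material for drifting mist on the Ghost Portal backpack cosmetic.
# Two panning inky-smoke samples, timed and offset by dynamic parameters, are warped by a world-space distortion normal map and summed.
# Opacity combines that smoke with a ripple mask, fading at grazing angles and near intersecting geometry.
# Emissive multiplies smoke by particle color and a brightened scene-color blend, which mobile replaces with a constant.
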Material /BRCosmetics/Accessories/FORT_Backpacks/M_MED_GhostPortal_Backpack/FX/M_Mist
graph {
 properties {
  bUsedWithNiagaraSprites=true
  bUsedWithParticleSprites=true
  blendMode=BLEND_Translucent
  pixelDepthOffsetMode=PDOM_Legacy
  shadingModel=MSM_Unlit
  usedShadingModels=Unlit
 }
 nodes [
  n1 = ParticleColor()
  n2 = DynamicParameter() {paramNames [Time,Random,Param3,Param4]}
  n3 = Add(A=TextureCoordinate() {uTiling=0.75, vTiling=0.15}, B=@n2.Random)
  n4 = Multiply(A=@n3) {constB=0.4}
  n5 = Add(A=@n2.Time, B=@n2.Time)
  n6 = WorldPosition()
  n7 = Constant() {r=256}
  n8 = Add(A=Divide(A=ComponentMask(None=@n6.XYZ) {g=true, r=true}, B=@n7), B=Multiply(A=Time()) {constB=0.1})
  n9 = Multiply(A=ComponentMask(None=TextureSample(UVs=@n8) {samplerType=SAMPLERTYPE_Normal, texture=&T_Weap_Distort_N_02}.RGB) {g=true, r=true}) {constB=0.1}
  n10 = TextureSample(UVs=Add(A=Panner(Coordinate=@n4, Time=@n5) {speedY=0.1}, B=@n9)) {texture=&T_Inky_Smoke_Tile_Chromatic}
  n11 = Panner(Coordinate=@n3, Time=@n5) {speedY=0.2}
  n12 = Add(
   A=@n10.RGB
   B=TextureSample(UVs=Add(A=@n11, B=@n9)) {texture=&T_Inky_Smoke_Tile_Chromatic}.RGB
  )
  n13 = TextureSample() {samplerType=SAMPLERTYPE_LinearColor, texture=/Game/Effects/Fort_Effects/Effects/Water/Textures/WaterDecalRippleMask}
  n14 = Multiply(A=Desaturation(None=@n12), B=@n13.R) {constB=0.25}
  n15 = CameraVectorWS()
  n16 = VertexNormalWS()
  n17 = Multiply(A=Multiply(A=@n14, B=@n14), B=Abs(None=DotProduct(A=@n15, B=@n16)))
  Add(A=Multiply(A=@n1.A, B=DepthFade(Opacity=@n17) {fadeDistanceDefault=60}), B=Multiply(A=Multiply(A=@n10.R, B=@n13.R)) {constB=0.02})
  n18 = Multiply(A=LinearInterpolate(B=SceneColor!(UVs).RGB) {constA=0.5, constAlpha=0.75}) {constB=2}
  n19 = Constant() {r=1}
  MaterialFunctionCall(Emissive=Multiply(A=ShadingPathSwitch(Default=@n18, Mobile=@n19) {desc=Migrated from FeatureLevelSwitch}, B=Multiply(A=@n12, B=@n1.RGB))) {materialFunction=&MF_FX_EmissiveAdjustments}
  MaterialFunctionCall!(TextureResolution=@n7) {materialFunction=/Engine/Functions/Engine_MaterialFunctions02/Texturing/ScreenAlignedPixelToPixelUVs}
 ]
}
