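# M_VotexWave is a translucent unlit material for the Ghost Portal backpack FX, drawing animated, refractive radial waves.
# A powered UV sphere mask plus time feeds sine and cosine, which drive a two-color blend, opacity variation and a procedural normal.
# Emissive optionally tints scene color, with a simpler mobile branch, while refraction scales with view angle and particle alpha.
# Opacity combines parameters, particle alpha, non-mobile depth fade and a camera-proximity fade.
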Material /BRCosmetics/Accessories/FORT_Backpacks/M_MED_GhostPortal_Backpack/FX/M_VotexWave
graph {
 properties {
  bUsedWithMeshDeformer=true
  bUsedWithSkeletalMesh=true
  blendMode=BLEND_Translucent
  pixelDepthOffsetMode=PDOM_Legacy
  refractionMethod=RM_IndexOfRefraction
  shadingModel=MSM_Unlit
  translucencyPass=MTP_BeforeDOF
  usedShadingModels=Unlit
 }
 nodes [
  n1 = SphereMask(A=TextureCoordinate(), B=Constant() {r=0.5}, Radius=Constant() {r=1}, Hardness=Constant())
  n2 = Multiply(A=@n1, B=@n1)
  n3 = Add(
   A=Multiply(A=Multiply(A=@n2, B=@n2), B=ScalarParameter() {defaultValue=1.5, parameterName=WaveNum}) {constB=1.5}
   B=Multiply(A=Time() {bOverride_Period=true, period=30}) {constB=0.3}
  )
  n4 = Sine(None=@n3)
  n5 = Cosine(None=@n3)
  n6 = Multiply(A=@n4, B=@n5)
  Add(B=@n6)
  n7 = DotProduct(A=CameraVectorWS(), B=VertexNormalWS())
  n8 = ScalarParameter() {defaultValue=1, parameterName=BP_Opacity}
  n9 = LinearInterpolate(Alpha=@n8) {constA=0.4}
  LinearInterpolate(
   A=Constant() {r=1}
   B=ScalarParameter() {defaultValue=1.02, parameterName=Ref}
   Alpha=Multiply(A=Multiply(A=@n7, B=@n9), B=ParticleColor().A)
  )
  n10 = LinearInterpolate(A=SceneColor!(UVs).RGB) {constAlpha=0.3, constB=0.5}
  n11 = LinearInterpolate(
   A=VectorParameter() {parameterName=Color2, defaultValue=<rgba 0 5.3784 6 1>}.RGB
   B=VectorParameter() {parameterName=Color1, defaultValue=<rgba 0 0.0989 0.5 1>}.RGB
   Alpha=@n6
  )
  n12 = Multiply(A=@n11) {constB=0.5}
  n13 = ShadingPathSwitch(Default=Multiply(A=StaticSwitchParameter(True=@n10, False=@n12) {defaultValue=true, parameterName=UseSceneColor}, B=@n11), Mobile=@n12) {desc=Migrated from FeatureLevelSwitch}
  MaterialFunctionCall(Emissive=Multiply(A=@n13, B=@n13)) {materialFunction=&MF_FX_EmissiveAdjustments}
  n14 = DepthFade() {fadeDistanceDefault=2}
  n15 = Constant() {r=1}
  n16 = Multiply(
   A=LinearInterpolate(Alpha=ConstantBiasScale(None=@n4) {scale=0.2}) {constA=0.5}
   B=Multiply(A=@n8, B=ShadingPathSwitch(Default=@n14, Mobile=@n15) {desc=Migrated from FeatureLevelSwitch})
  )
  n17 = ParticleColor()
  n18 = WorldPosition()
  n19 = CameraPositionWS()
  n20 = Constant() {r=150}
  n21 = Constant() {r=0.5}
  Multiply(
   A=Multiply(A=ScalarParameter() {defaultValue=1, parameterName=OverallOpacity}, B=Multiply(A=@n16, B=@n17.A))
   B=OneMinus(None=SphereMask(A=@n18.XYZ, B=@n19, Radius=@n20, Hardness=@n21))
  )
  Multiply()
  n22 = AppendVector(A=@n4, B=@n5)
  n23 = Subtract(A=OneMinus(None=Abs(None=@n22))) {constB=0.5}
  n24 = Constant() {r=0.5}
  Normalize(VectorInput=LinearInterpolate(A=Constant3Vector() {constant=<rgba 0 0 1 0>}, B=AppendVector(A=@n23, B=@n24), Alpha=@n1))
 ]
}
