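# Unlit, two-sided translucent mesh-particle material for the wave effect on the GhostPortal backpack.
# The dynamic parameter Time pans a clamped smash-trail mask along U, allowing particle-driven sweeps.
# That squared mask multiplies two smoke-trace samples, one with swapped axes and square-root-warped U.
# Particle color tints the result through MF_FX_EmissiveAdjustments, while particle alpha scales a saturated copy, apparently for opacity.
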Material /BRCosmetics/Accessories/FORT_Backpacks/M_MED_GhostPortal_Backpack/FX/M_Waves
graph {
 properties {
  bUsedWithMeshParticles=true
  bUsedWithNiagaraMeshParticles=true
  blendMode=BLEND_Translucent
  pixelDepthOffsetMode=PDOM_Legacy
  shadingModel=MSM_Unlit
  twoSided=true
  usedShadingModels=Unlit
 }
 nodes [
  n1 = MaterialFunctionCall(Float2=TextureCoordinate()) {materialFunction=&BreakOutFloat2Components}
  n2 = TextureSample(UVs=AppendVector(A=Multiply(A=@n1.R) {constB=0.5}, B=@n1.G)) {texture=/Game/Effects/Fort_Effects/Textures/Smoke/T_ShotGun_Trace_Packed}
  n3 = Cosine(None=ComponentMask(None=TextureCoordinate() {uTiling=0.5, vTiling=0.5}) {g=true})
  LinearInterpolate(A=@n2.R, B=OneMinus(None=Abs(None=@n3)))
  n4 = TextureSample(UVs=Panner(Time=DynamicParameter() {paramNames [Time,Param2,Param3,Param4]}.Time) {speedX=1}) {samplerSource=SSM_Clamp_WorldGroupSettings, texture=/Game/Effects/Fort_Effects/Textures/Melee/T_SmashTrail}
  n5 = MaterialFunctionCall(Float2=TextureCoordinate()) {materialFunction=&BreakOutFloat2Components}
  n6 = SquareRoot(None=@n5.R)
  n7 = Multiply(A=TextureSample(UVs=AppendVector(A=@n5.G, B=@n6)) {texture=&T_ShotGun_Trace_Packed}.R) {constB=8}
  n8 = Multiply(A=Multiply(A=@n4.R, B=@n4.R), B=Multiply(A=Saturate(None=@n7), B=@n2.R))
  n9 = ParticleColor()
  MaterialFunctionCall(Emissive=Multiply(A=Multiply(A=@n8), B=@n9.RGB)) {materialFunction=&MF_FX_EmissiveAdjustments}
  Multiply(A=@n1.G) {constB=2}
  Multiply(A=Saturate(None=Multiply(A=@n8) {constB=5}), B=@n9.A) {constB=0.5}
 ]
}
