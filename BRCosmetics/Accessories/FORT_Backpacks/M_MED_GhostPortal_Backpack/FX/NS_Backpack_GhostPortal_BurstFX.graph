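# One-shot burst effect for the GhostPortal backpack cosmetic, combining a portal flash with ghost sprites thrown outward.
# DarkGradient_Packed packs seven spawn groups into one emitter, using VisibilityTag switches to give its sprite, light and mesh layers separate lifetimes and motion, including delayed sparks.
# S3_Ghost briefly spawns SubUV sprites that grow, fade and drift under drag and point attraction.
# FN_VariantManager and VariantColorSelector inputs provide blue, purple and green colour variants, exposed through User.Variant.
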NiagaraSystem /BRCosmetics/Accessories/FORT_Backpacks/M_MED_GhostPortal_Backpack/FX/NS_Backpack_GhostPortal_BurstFX
dependencies {
 usedDataInterfaces [
  &NiagaraDataInterfaceArrayFloat
  &NiagaraDataInterfaceArrayInt32
  &NiagaraDataInterfaceArrayFloat2
  &NiagaraDataInterfaceVectorCurve
  &NiagaraDataInterfaceCurve
  &NiagaraDataInterfaceSpriteRendererInfo
 ]
 usedDynamicInputs []
 usedModules [&SystemState, &FN_VariantManager]
}
emitters {
 DarkGradient_Packed {
  emitterUpdateScript [
   &EmitterState(Life Cycle Mode=::NewEnumerator1, Loop Behavior=::NewEnumerator1, Loop Duration=1, UseLoopDelay=false)
   &SpawnBurst_Instantaneous(Use Spawn Probability=false, UseLoopCountLimit=false)
   &SpawnBurst_Instantaneous(Spawn Group=1, Use Spawn Probability=false, UseLoopCountLimit=false) {module=SpawnBurst_Instantaneous002}
   /Game/Effects/Niagara/Modules/Spawn/FN_Spawn_TriggeredBurst(
    Burst Count Limit=1
    Module.Use Burst Count Limit=true
    Spawn Count:NiagaraInt32=&RandomRangeInt(Minimum=20, Maximum=30)
    Spawn Group=2
    Spawn Trigger Mode=2
    Timer=0.35
    UseLoopCountLimit=true
   )
   &SpawnBurst_Instantaneous(Spawn Group=3, Use Spawn Probability=false, UseLoopCountLimit=false) {module=SpawnBurst_Instantaneous001}
   &SpawnBurst_Instantaneous(Spawn Count=2, Spawn Group=4, Use Spawn Probability=false, UseLoopCountLimit=false) {module=SpawnBurst_Instantaneous003}
   &SpawnBurst_Instantaneous(Spawn Group=5, Use Spawn Probability=false, UseLoopCountLimit=false) {module=SpawnBurst_Instantaneous004}
   &SpawnBurst_Instantaneous(Spawn Count=4, Spawn Group=6, Use Spawn Probability=false, UseLoopCountLimit=false) {module=SpawnBurst_Instantaneous005}
  ]
  particleSpawnScript [
   /SetVariables(Particles.VisibilityTag=$Emitter.SpawnGroup) {isSetParameters=true}
   &InitializeParticle(
    Color=&VariantColorSelector(
     Variant Amount=1
     Original Color=<rgba 0.0357 0.119 0.188 1>
     Variant 1 Color=<rgba 0.0042 0 0.188 0.85>
     Variant 2 Color=<rgba 0 0.188 0.0501 0.5>
    )
    Lifetime=&MergeHelper_FloatSwitch(
     0=0.85
     1=0.35
     2=&RandomRangeFloat(Minimum=0.2, Maximum=1, Evaluation Type=::NewEnumerator1)
     3=0.4
     4=0.1
     5=0.5
     6=&RandomRangeFloat(Minimum=0.4, Maximum=0.3, Evaluation Type=::NewEnumerator1)
     Index=$Particles.VisibilityTag
     Num=::NewEnumerator5
    )
    Mesh Scale=<0.27 0.3 0.3>
    Mesh Scale Mode=::NewEnumerator4
    Position Offset=&MergeHelper_VectorSwitch(
     0=<-5 -10 1>
     1=<-5 -20 1>
     2=<-5 -10 1>
     3=<-5 -12 1>
     4=<-5 -10 1>
     5=<-5 -5 1>
     6=<-5 -10 1>
     Index=$Particles.VisibilityTag
     Num=::NewEnumerator5
    )
    Sprite Rotation Mode=::NewEnumerator2
    Sprite Size Mode=::NewEnumerator3
    Sprite UV Mode=::NewEnumerator3
    Uniform Sprite Size=175
    UsePositionOffset=true
    Write Material Random=true
   )
   &CameraOffset(Camera Offset Amount=&MergeHelper_FloatSwitch(0=15, 1=0, 2=0, 3=20, 4=25, 5=20, 6=0, Index=$Particles.VisibilityTag, Num=::NewEnumerator5))
   &DynamicMaterialParameters(
    Dynamic Parameter Output:ENiagara_AttributeSamplingApplyOutput=::NewEnumerator1
    Index 0 Param 1=4
    Index 0 Param 2=1
    Index 0 Param 3=1
    Index 0 Param 4=1
    Module.Index0Float0Write=true
    Module.Index0Float1Write=true
    Module.Index0Float2Write=true
    Module.Index0Float3Write=true
   )
   &DynamicMaterialParameters(
    Dynamic Parameter Output:ENiagara_AttributeSamplingApplyOutput=::NewEnumerator1
    Index 0 Param 1=1
    Index 0 Param 2=1
    Index 0 Param 3=1
    Index 0 Param 4=1
    Module.Index0Float0Write=true
    Module.Index0Float1Write=true
    Module.Index0Float2Write=true
    Module.Index0Float3Write=true
   ) {module=DynamicMaterialParameters001}
   &ShapeLocation(
    MaskSpawnGroup=false
    Override Seed=false
    Sphere Radius=&MergeHelper_FloatSwitch(0=0, 1=0, 2=5, 3=0, 4=0, 5=0, 6=3, Index=$Particles.VisibilityTag, Num=::NewEnumerator5)
    UseOwnerScale=true
   )
   &AddVelocity(
    Cone Angle=&RandomRangeFloat(Minimum=0, Maximum=0.25, Evaluation Type=::NewEnumerator1)
    Cone Angle Mode:ENiagara_AngleInput=::NewEnumerator1
    Cone Axis=<0 -1 0>
    Module.UseConeFalloff=false
    Module.UseVelDistribution=false
    Override Seed=false
    Velocity Mode=::NewEnumerator2
    Velocity Speed=&MergeHelper_FloatSwitch(
     0=0
     1=0
     2=&RandomRangeFloat(Minimum=0, Maximum=125, Evaluation Type=::NewEnumerator1)
     3=0
     4=0
     5=0
     6=0
     Index=$Particles.VisibilityTag
     Num=::NewEnumerator5
    )
   )
   &InitialMeshOrientation(
    Mesh Orientation Mode=::NewEnumerator7
    Rotation=&MaskVector(
     Mask Source=0
     Mask Parameter=?
     Mask=6
     Masked Vector=&RandomRangeVector(Minimum=<0 0 0>, Maximum=<0 1 0>, Evaluation Type=::NewEnumerator1)
     Passthrough Vector=&MakeVector(X=0, Y=&RandomRangeFloat(Minimum=0, Maximum=1), Z=0)
    )
    Use Rotation Vector=true
   )
  ]
  particleUpdateScript [
   &ParticleState()
   &SpriteRotationRate(Rotation Rate=&MergeHelper_FloatSwitch(
     0=0
     1=0
     2=0
     3=0
     4=&RandomRangeFloat(Minimum=-0.2, Maximum=0.2)
     5=0
     6=0
     Index=$Particles.VisibilityTag
     Num=::NewEnumerator5
    ))
   &ScaleMeshSize(Scale Factor=&VectorFromCurve(VectorCurve=DataInterface() {propertyValues {
       XCurve {defaultValue=1.1, keys [<at 0 0.6>,<at 1 0.2095>]}
       YCurve {defaultValue=1, keys [<at 0 1.1>,<at 1 0.5466>]}
       ZCurve {defaultValue=0.6, keys [<at 0 1.1>,<at 1 0.5466>]}
      }}))
   &ScaleColor(
    Scale Alpha=&FloatFromCurve(FloatCurve=DataInterface() {propertyValues {Curve {keys [
         {interpMode=RCIM_Cubic, leaveTangentWeight=0.1333, tangentMode=RCTM_User, time=0, value=0}
         {arriveTangentWeight=0.2592, interpMode=RCIM_Cubic, leaveTangentWeight=0.1296, tangentMode=RCTM_User, time=0.4, value=1}
         {arriveTangentWeight=0.0667, interpMode=RCIM_Cubic, leaveTangentWeight=0.1333, tangentMode=RCTM_User, time=0.6, value=1}
         <at 1 0>
        ]}}})
    ScaleA=true
    ScaleRGB=true
   )
   &FN_Light_Attributes(
    Color=&MakeLinearColorFromVectorAndFloat(
     "Vector (RGB)"=&MakeVectorFromLinearColorRGB(Color=&VariantColorSelector(Variant Amount=1, Original Color=<rgba 0.06 5.11 10 1>, Variant 1 Color=<rgba 2.48 0.4116 10 1>, Variant 2 Color=<rgba 1.4583 10 3.7329 1>))
     "Float (Alpha)"=&FloatFromCurve(FloatCurve=DataInterface() {propertyValues {Curve {keys [
          <at 0 0 RCIM_Cubic>
          {arriveTangent=0.2214, interpMode=RCIM_Cubic, leaveTangent=0.2214, time=0.2451, value=0.1268}
          {arriveTangent=0.6864, interpMode=RCIM_Cubic, leaveTangent=0.6864, time=0.3915, value=0.0867}
          {arriveTangent=4.1126, interpMode=RCIM_Cubic, leaveTangent=4.1126, time=0.5622, value=0.3444}
          {arriveTangent=-0.7349, interpMode=RCIM_Cubic, leaveTangent=-0.7349, time=0.6136, value=1}
          {arriveTangent=-2.5881, interpMode=RCIM_Cubic, leaveTangent=-2.5881, time=0.7509, value=0.2057}
          <at 1 0 RCIM_Cubic>
         ]}}})
    )
    Create Color Attribute=true
    Create Exponent Attribute=true
    Create Radius Attribute=true
    Exponent=4
    Radius=90
    Scale Color=&FloatFromCurve(FloatCurve=DataInterface() {propertyValues {Curve {defaultValue=1, keys [<at 0 1 RCIM_Cubic>,<at 1 0.75 RCIM_Cubic>]}}})
   )
   &Drag(Drag=3, Use Linear Drag=true, Use Rotational Drag=true)
   &FN_Sprite_Attributes(
    Color=&MakeLinearColorFromVectorAndFloat(
     "Vector (RGB)"=&MakeVectorFromLinearColorRGB(Color=&VariantColorSelector(Variant Amount=1, Original Color=<rgba 2.1266 12.2912 16 1>, Variant 1 Color=<rgba 7.4199 2.4 16 1>, Variant 2 Color=<rgba 1.6 16 5.4345 1>))
     "Float (Alpha)"=&FloatFromCurve(FloatCurve=DataInterface() {propertyValues {Curve {keys [<at 0 1>,<at 0.2 0.2>]}}})
    )
    Create Color Attribute=true
    Create SpriteSize Attribute=true
    Module Output:ENiagara_AttributeSamplingApplyOutput=::NewEnumerator1
    SpriteSize=&Vector2DFromFloat(Value=&Multiply_Float(
      A=&RandomRangeFloat(Minimum=3, Maximum=7)
      B=&FloatFromCurve(FloatCurve=DataInterface() {propertyValues {Curve {keys [
           <at -0.0044 0.2654>
           {arriveTangent=0.7558, interpMode=RCIM_Cubic, leaveTangent=0.7558, time=0.157, value=0.5261}
           {arriveTangent=-0.6241, interpMode=RCIM_Cubic, leaveTangent=-0.6241, time=0.7034, value=0.8003}
           <at 1 0>
          ]}}})
     ))
   )
   /Niagara/Modules/CascadeConversion/CascadeConversion_Orbit(
    Offset:Vector3f=&VectorFromCurve(VectorCurve=DataInterface() {propertyValues {
       XCurve {keys [{interpMode=RCIM_Cubic, tangentMode=RCTM_User, time=0, value=0}, <at 1 0 RCIM_Cubic>]}
       YCurve {keys [{arriveTangent=-10.1229, interpMode=RCIM_Cubic, leaveTangent=-10.1229, tangentMode=RCTM_User, time=0, value=0}, <at 1 -10 RCIM_Cubic>]}
       ZCurve {keys [{arriveTangent=10.0198, interpMode=RCIM_Cubic, leaveTangent=10.0198, tangentMode=RCTM_User, time=0, value=3}, <at 1 12 RCIM_Cubic>]}
      }})
    Rotation:Vector3f=&RandomRangeVector(Minimum=<0 0 0>, Maximum=<0 1 0>)
    Rotation Rate:Vector3f=&RandomRangeVector(Minimum=<0 0.05 0>, Maximum=<0 0.1 0>)
   )
   /Niagara/Modules/CascadeConversion/CascadeConversion_SolveOrbit()
   &FN_Mesh_Attributes(
    Color=&MakeLinearColorFromVectorAndFloat(
     "Vector (RGB)"=&Multiply_Vector(
      A=&MakeVectorFromLinearColorRGB(Color=&VariantColorSelector(Variant Amount=1, Original Color=<rgba 0.217 1.5542 2 1>, Variant 1 Color=<rgba 0.8935 0.6 4 1>, Variant 2 Color=<rgba 0.2917 2 0.7466 1>))
      B=&VectorFromCurve(VectorCurve=DataInterface() {propertyValues {XCurve {keys [<at 0 2>,<at 1 0.985>]}, YCurve {keys [<at 0 2>,<at 1 0.6599>]}, ZCurve {keys [<at 0 2>,<at 1 1>]}}})
     )
     "Float (Alpha)"=&FloatFromCurve(
      FloatCurve=DataInterface() {propertyValues {Curve {keys [
          <at 0 0 RCIM_Cubic>
          {arriveTangent=3.3714, interpMode=RCIM_Cubic, leaveTangent=-3.8431, tangentMode=RCTM_Break, time=0.4832, value=0.7756}
          <at 1 0>
         ]}}}
      Scale Curve=0.15
     )
    )
    Create Color Attribute=true
    Module Output=::NewEnumerator1
   )
   &DynamicMaterialParameters(
    Dynamic Parameter Output:ENiagara_AttributeSamplingApplyOutput=::NewEnumerator1
    Index 0 Param 1=-1.5
    Index 0 Param 2=1
    Index 0 Param 3=1
    Index 0 Param 4=1
    Module.Index0Float0Write=true
    Module.Index0Float1Write=true
    Module.Index0Float2Write=true
    Module.Index0Float3Write=true
   ) {module=DynamicMaterialParameters002}
   &FN_Sprite_Attributes(
    Color=&MultiplyLinearColor(
     Linear Color=&VariantColorSelector(
      Variant Amount=1
      Original Color=<rgba 0 0.7225 1 0.2>
      Variant 1 Color=<rgba 0.6258 0.195 1.5 0.2>
      Variant 2 Color=<rgba 0.1407 0.9647 0.3601 0.2>
     )
     Scale Factor=&MakeVector4FromVectorAndFloat(
      "Vector (RGB)"=&VectorFromFloat(Value=&FloatFromCurve(FloatCurve=DataInterface() {propertyValues {Curve {keys [<at 0 1>,<at 1 0>]}}}))
      "Float (Alpha)"=&RandomRangeFloat(Minimum=0, Maximum=1, Evaluation Type=::NewEnumerator1)
     )
    )
    Create Color Attribute=true
    Create Sprite Rotation Attribute=true
    Create SpriteSize Attribute=true
    Module Output:ENiagara_AttributeSamplingApplyOutput=::NewEnumerator1
    SpriteRotation:NiagaraFloat=&RandomRangeFloat(Minimum=0, Maximum=360)
    SpriteSize=&Multiply_Vector2D_ByFloat(
     Float=&FloatFromCurve(FloatCurve=DataInterface() {propertyValues {Curve {keys [{arriveTangent=-0.9923, time=0, value=1}, <at 1 0.0077>]}}})
     Vector2D=<300 40>
    )
   ) {module=FN_Sprite_Attributes001}
   &FN_Sprite_Attributes(
    Color=&MultiplyLinearColor(
     Linear Color=&VariantColorSelector(
      Variant Amount=1
      Original Color=<rgba 0 0.7225 1 0.2>
      Variant 1 Color=<rgba 0.6258 0.195 1.5 0.2>
      Variant 2 Color=<rgba 0.1407 0.9647 0.3601 0.2>
     )
     Scale Factor=&MakeVector4FromVectorAndFloat(
      "Vector (RGB)"=&VectorFromFloat(Value=&FloatFromCurve(FloatCurve=DataInterface() {propertyValues {Curve {keys [{arriveTangent=-60.9428, interpMode=RCIM_Cubic, leaveTangent=-60.9428, tangentMode=RCTM_User, time=0, value=20}, <at 1 0 RCIM_Cubic>]}}}))
      "Float (Alpha)"=&RandomRangeFloat(Minimum=0, Maximum=1, Evaluation Type=::NewEnumerator1)
     )
    )
    Create Color Attribute=true
    Create Sprite Rotation Attribute=true
    Create SpriteSize Attribute=true
    Module Output:ENiagara_AttributeSamplingApplyOutput=::NewEnumerator1
    SpriteRotation:NiagaraFloat=&RandomRangeFloat(Minimum=-360, Maximum=36)
    SpriteSize=&Multiply_Vector2D_ByFloat(
     Float=&FloatFromCurve(FloatCurve=DataInterface() {propertyValues {Curve {keys [<at 0 1>,<at 1 2>]}}})
     Vector2D=<65 65>
    )
   ) {module=FN_Sprite_Attributes002}
   &DynamicMaterialParameters(
    Dynamic Parameter Output:ENiagara_AttributeSamplingApplyOutput=::NewEnumerator1
    Index 0 Param 1=&FloatFromCurve(FloatCurve=DataInterface() {propertyValues {Curve {keys [
         {arriveTangent=-0.0157, interpMode=RCIM_Cubic, leaveTangent=-0.0157, tangentMode=RCTM_User, time=0, value=-1.0027}
         {arriveTangent=1.2932, interpMode=RCIM_Cubic, leaveTangent=1.2932, time=0.2574, value=-0.1899}
         <at 1 0.2905>
        ]}}})
    Index 0 Param 2=1
    Index 0 Param 3=1
    Index 0 Param 4=1
    Module.Index0Float0Write=true
    Module.Index0Float1Write=true
    Module.Index0Float2Write=true
    Module.Index0Float3Write=true
   ) {module=DynamicMaterialParameters003}
   &FN_Mesh_Attributes(
    Color=&MultiplyLinearColor(
     Linear Color=&VariantColorSelector(
      Variant Amount=1
      Original Color=<rgba 0.026 0.578 1 1>
      Variant 1 Color=<rgba 0.6208 0.1875 1.5 1>
      Variant 2 Color=<rgba 0.0833 1 0.3274 1>
     )
     Scale Factor=&MakeVector4FromVectorAndFloat(
      "Vector (RGB)"=&VectorFromCurve(VectorCurve=DataInterface() {propertyValues {
         XCurve {keys [<at 0.4 1 RCIM_Cubic>,<at 0.7 10 RCIM_Cubic>]}
         YCurve {keys [<at 0.4 1 RCIM_Cubic>,<at 0.7 10 RCIM_Cubic>]}
         ZCurve {keys [<at 0.4 1 RCIM_Cubic>,<at 0.7 10 RCIM_Cubic>]}
        }})
      "Float (Alpha)"=&FloatFromCurve(FloatCurve=DataInterface() {propertyValues {Curve {keys [
           <at 0 0 RCIM_Cubic>
           {arriveTangent=2.8589, interpMode=RCIM_Cubic, leaveTangent=2.8589, tangentMode=RCTM_User, time=0.8325, value=0.98}
           <at 1 0 RCIM_Cubic>
          ]}}})
     )
    )
    Create Color Attribute=true
    Create Scale Attribute=true
    Module Output=::NewEnumerator1
    Scale=&Multiply_Vector(
     A=&RandomRangeVector(Minimum=<0.3 0.3 2>, Maximum=<0.45 0.45 4>)
     B=&VectorFromCurve(VectorCurve=DataInterface() {propertyValues {
        XCurve {keys [<at 0 1 RCIM_Cubic>,<at 1 0.1965 RCIM_Cubic>]}
        YCurve {keys [<at 0 1 RCIM_Cubic>,<at 1 0.7359 RCIM_Cubic>]}
        ZCurve {keys [<at 0 1 RCIM_Cubic>,<at 1 0.7359 RCIM_Cubic>]}
       }})
    )
   ) {module=FN_Mesh_Attributes001}
   &SolveForcesAndVelocity(Clamp Velocity=false, Limit Acceleration=false)
  ]
  rendererClasses [
   &NiagaraSpriteRendererProperties
   &NiagaraLightRendererProperties
   &NiagaraMeshRendererProperties
  ]
  renderers [
   {rendererClass=&NiagaraSpriteRendererProperties}
   {rendererClass=&NiagaraLightRendererProperties, rendererIndex=1}
   {rendererClass=&NiagaraSpriteRendererProperties, rendererIndex=2}
   {rendererClass=&NiagaraMeshRendererProperties, rendererIndex=3}
   {rendererClass=&NiagaraSpriteRendererProperties, rendererIndex=4}
   {rendererClass=&NiagaraSpriteRendererProperties, rendererIndex=5}
   {rendererClass=&NiagaraMeshRendererProperties, rendererIndex=6}
  ]
 }
 S3_Ghost {
  emitterUpdateScript [
   &EmitterState(Life Cycle Mode=::NewEnumerator1, Loop Behavior=::NewEnumerator1, Loop Delay=0.05, Loop Duration=0.6, UseLoopDelay=true)
   &SpawnRate(SpawnRate=6, Use Spawn Probability=false)
   &SpawnRate(SpawnGroup=1, SpawnRate=4, Use Spawn Probability=false) {module=SpawnRate001}
  ]
  particleSpawnScript [
   /SetVariables(Particles.VisibilityTag=$Emitter.SpawnGroup) {isSetParameters=true}
   &InitializeParticle(
    Color=&RandomRangeLinearColor(Minimum=<rgba 8 8 8 2>, Maximum=<rgba 4 4 4 4>)
    Lifetime=&MergeHelper_FloatSwitch(
     0=&RandomRangeFloat(Minimum=1, Maximum=2, Evaluation Type=::NewEnumerator1)
     1=&RandomRangeFloat(Minimum=0.5, Maximum=1)
     2=?
     3=?
     Index=$Particles.VisibilityTag
     Num=::NewEnumerator0
    )
    Position Offset=<-5 -3 1>
    Sprite Size=&RandomRangeVector2D(Minimum=<12 24>, Maximum=<16 35>, Evaluation Type=::NewEnumerator1)
    Sprite Size Mode=::NewEnumerator4
    UsePositionOffset=true
    Write Material Random=true
   )
   &AddVelocity(Velocity=&RandomRangeVector(Minimum=<-20 -40 -20>, Maximum=<20 -200 20>, Evaluation Type=::NewEnumerator1))
   &SubUVAnimation(
    End Frame Range Override=5
    Evaluation Type=::NewEnumerator0
    Output Alternate Renderer Binding=true
    "Sprite Renderer "=DataInterface() {propertyValues {SpriteRenderer=&self:S3_Ghost_0.NiagaraSpriteRendererProperties_1}}
    SubUV Animation Mode:ENiagara_SubUVLookupModeV2=::NewEnumerator2
    UseEndFrame=false
    UseStartFrame=false
   )
  ]
  particleUpdateScript [
   &ParticleState()
   &ScaleSpriteSize(
    Scale Factor=&Vector2DFromFloat(Value=&FloatFromCurve(FloatCurve=DataInterface() {propertyValues {Curve {keys [{arriveTangent=3.006, interpMode=RCIM_Cubic, leaveTangent=3.006, tangentMode=RCTM_User, time=0, value=0}, <at 1 1 RCIM_Cubic>]}}}))
    Scale Sprite Size Mode=::NewEnumerator2
   )
   /Niagara/Modules/Update/Forces/PointAttractionForce(
    Attraction Radius=150
    AttractionStrength=30
    AttractorPosition:Vector3f=&RandomRangeVector(Minimum=<30 0 0>, Maximum=<60 0 0>, Evaluation Type=::NewEnumerator1)
    Falloff Exponent=0.1
    Module.Kill Within Radius=false
    Module.Use Falloff=true
   )
   &AccelerationForce(Acceleration=&RandomRangeVector(Minimum=<0 0 0>, Maximum=<0 0 50>))
   &Drag(Drag=2, Use Linear Drag=true, Use Rotational Drag=true)
   &SolveForcesAndVelocity(Clamp Velocity=false, Limit Acceleration=false)
   &ScaleColor(
    Scale Alpha=&FloatFromCurve(FloatCurve=DataInterface() {propertyValues {Curve {keys [<at 0 0 RCIM_Cubic>,<at 0.1044 1 RCIM_Cubic>,<at 1 0 RCIM_Cubic>]}}})
    Scale RGB=&MakeVectorFromLinearColorRGB(Color=&VariantColorSelector(
      Variant Amount=1
      Original Color=<rgba 0.152 0.607 1 1>
      Variant 1 Color=<rgba 0.369 0.0581 1 1>
      Variant 2 Color=<rgba 0.2874 0.9647 0.4678 1>
     ))
    ScaleA=true
    ScaleRGB=true
   )
   &DynamicMaterialParameters(
    Dynamic Parameter Output:ENiagara_AttributeSamplingApplyOutput=::NewEnumerator1
    Index 0 Param 1=1
    Index 0 Param 2=1
    Index 0 Param 3=&FloatFromCurve(FloatCurve=DataInterface() {propertyValues {Curve {keys [<at 0 0 RCIM_Cubic>,<at 0.3126 0.1502 RCIM_Cubic>,<at 1 0 RCIM_Cubic>]}}})
    Index 0 Param 4=&FloatFromCurve(FloatCurve=DataInterface() {propertyValues {Curve {keys [{arriveTangent=-29.0081, interpMode=RCIM_Cubic, leaveTangent=-29.0081, tangentMode=RCTM_User, time=0, value=-5.1116}, <at 1 -15 RCIM_Cubic>]}}})
    Module.Index0Float0Write=true
    Module.Index0Float1Write=true
    Module.Index0Float2Write=true
    Module.Index0Float3Write=true
   )
  ]
  rendererClasses [&NiagaraSpriteRendererProperties]
  renderers [
   {rendererClass=&NiagaraSpriteRendererProperties}
   {rendererClass=&NiagaraSpriteRendererProperties, rendererIndex=1}
  ]
 }
}
summary {
 systemName=NS_Backpack_GhostPortal_BurstFX
 emitters [
  {
   emitterName=DarkGradient_Packed
   rendererClasses [
    &NiagaraSpriteRendererProperties
    &NiagaraLightRendererProperties
    &NiagaraMeshRendererProperties
   ]
  }
  {emitterName=S3_Ghost, rendererClasses [&NiagaraSpriteRendererProperties]}
 ]
 userVariables [User.Variant {type=NiagaraFloat, defaultValue=0, description=""}]
}
