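# Niagara idle effect for the GhostPortal backpack cosmetic, layering a glowing portal, mist and drifting ghost sprites.
# DarkGradient_Packed packs sprite, mesh and light layers into one emitter, using spawn groups copied to VisibilityTag to switch per-layer settings.
# S2_MistParticle and S3_Ghost also spawn with movement and inherit velocity.
# User.Stage gates some portal and mist spawners, variant selectors choose among three color schemes, and SystemFade scales several portal alphas.
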NiagaraSystem /BRCosmetics/Accessories/FORT_Backpacks/M_MED_GhostPortal_Backpack/FX/NS_Backpack_GhostPortal_IdleFX
dependencies {
 usedDataInterfaces [
  &NiagaraDataInterfaceArrayFloat
  &NiagaraDataInterfaceArrayInt32
  &NiagaraDataInterfaceArrayFloat2
  &NiagaraDataInterfaceCurve
  &NiagaraDataInterfaceVectorCurve
  &NiagaraDataInterfaceSpriteRendererInfo
  &NiagaraDataInterfaceVector2DCurve
 ]
 usedDynamicInputs [&FN_FrontEndFloatModifier]
 usedModules [
  &SystemState
  &self:SystemSpawnScript.SystemScriptSource.SystemScriptGraph.NiagaraNodeAssignment_1.SetVariables_#1
  &self:SystemSpawnScript.SystemScriptSource.SystemScriptGraph.NiagaraNodeAssignment_0.SetVariables_#2
  &FN_VariantManager
 ]
}
emitters {
 DarkGradient_Packed {
  emitterUpdateScript [
   &EmitterState(Inactive Response=::NewEnumerator1, Life Cycle Mode=::NewEnumerator1, Loop Duration=1, UseLoopDelay=true)
   &SpawnBurst_Instantaneous(Use Spawn Probability=false, UseLoopCountLimit=false)
   &SpawnRate(
    Spawn Probability=&MakeCustomFloatFromBool(
     Bool=&Boolean_OrOperation(
      Bool1=&SetBoolByIntComparison(A=&MakeIntFromFloat(Float=$User.Stage), B=0, Comparison Type=::NewEnumerator4)
      Bool2=&SetBoolByIntComparison(A=&MakeIntFromFloat(Float=$User.Stage), B=1, Comparison Type=::NewEnumerator4)
     )
     True Float=0
     False Float=1
    )
    SpawnGroup=1
    Use Spawn Probability=true
   )
   &SpawnRate(SpawnGroup=2, SpawnRate=2, Use Spawn Probability=false) {module=SpawnRate001}
   /Game/Effects/Niagara/Modules/Spawn/FN_Spawn_TriggeredBurst(Cooldown=2, Module.Use Burst Count Limit=false, Spawn Group=3, UseLoopCountLimit=false)
   &SpawnBurst_Instantaneous(Spawn Group=4, Use Spawn Probability=false, UseLoopCountLimit=false) {module=SpawnBurst_Instantaneous001}
   &SpawnRate(
    SpawnGroup=5
    SpawnRate=&Multiply_Float(
     A=5
     B=&FloatFromCurve(
      FloatCurve=DataInterface() {propertyValues {Curve {keys [
          <at 0 1 RCIM_Cubic>
          {arriveTangent=-0.2543, interpMode=RCIM_Cubic, leaveTangent=-0.2543, time=1.2828, value=1.6362}
          {arriveTangent=-0.3703, interpMode=RCIM_Cubic, leaveTangent=-0.3703, time=2.4472, value=0.3776}
          {arriveTangent=-0.11, interpMode=RCIM_Cubic, leaveTangent=-0.11, time=3.6302, value=0.7671}
          {arriveTangent=0.2907, interpMode=RCIM_Cubic, leaveTangent=0.2907, time=5.034, value=0.0931}
          {arriveTangent=0.3058, interpMode=RCIM_Cubic, leaveTangent=0.3058, time=6.8459, value=1.702}
          <at 8 1>
         ]}}}
      CurveIndex=$Emitter.NormalizedLoopAge
     )
    )
    Use Spawn Probability=false
   ) {module=SpawnRate002}
   /SetVariables(
    Emitter.MeshTrailColor:LinearColor=&VariantColorSelector(
     Variant Amount=1
     Original Color=<rgba 0.1137 0.5631 1 1>
     Variant 1 Color=<rgba 0.1845 0.0405 1 1>
     Variant 2 Color=<rgba 0.2874 0.9647 0.4678 1>
    )
    Emitter.MeshVortexColorA:LinearColor=&VariantColorSelector(Variant Amount=1, Original Color=<rgba 0 2.5 4 1>, Variant 1 Color=<rgba 0.7143 0.3827 4 1>, Variant 2 Color=<rgba 0.0926 3 1.7661 1>)
    Emitter.MeshVortexColorB:LinearColor=&VariantColorSelector(Variant Amount=1, Original Color=<rgba 0 10.2419 12 1>, Variant 1 Color=<rgba 3.2004 0.4444 12 1>, Variant 2 Color=<rgba 0 10 5.9665 1>)
   ) {isSetParameters=true}
  ]
  particleSpawnScript [
   /SetVariables(Particles.VisibilityTag=$Emitter.SpawnGroup) {isSetParameters=true}
   &InitializeParticle(
    Color=&VariantColorSelector(
     Variant Amount=1
     Original Color=<rgba 0.0357 0.119 0.188 1>
     Variant 1 Color=<rgba 0.0042 0 0.188 0.85>
     Variant 2 Color=<rgba 0 0.188 0.0501 0.5>
    )
    Lifetime=&MergeHelper_FloatSwitch(
     0=1.7
     1=&RandomRangeFloat(Minimum=2, Maximum=3)
     2=3
     3=2
     4=&RandomRangeFloat(Minimum=2.7, Maximum=3.2, Evaluation Type=::NewEnumerator1)
     5=&RandomRangeFloat(Minimum=2, Maximum=4, Evaluation Type=::NewEnumerator1)
     Index=$Particles.VisibilityTag
     Num=::NewEnumerator4
    )
    Mesh Scale=<1 0.5 1>
    Mesh Scale Mode=::NewEnumerator4
    Position Offset=&MergeHelper_VectorSwitch(0=<-5 -10 1>, 1=<-5 -8 1>, 2=<-5 2 1>, 3=<-5 -20 1>, 4=<-5 -10 1>, 5=<-5 -10 1>, Index=$Particles.VisibilityTag, Num=::NewEnumerator4)
    Sprite Rotation Mode=::NewEnumerator2
    Sprite Size Mode=::NewEnumerator3
    Sprite UV Mode=::NewEnumerator3
    Uniform Sprite Size=125
    UsePositionOffset=true
    Write Material Random=true
   )
   &CameraOffset(Camera Offset Amount=&MergeHelper_FloatSwitch(0=15, 1=0, 2=0, 3=0, 4=20, 5=0, Index=$Particles.VisibilityTag, Num=::NewEnumerator4))
   &DynamicMaterialParameters(
    Dynamic Parameter Output:ENiagara_AttributeSamplingApplyOutput=::NewEnumerator1
    Index 0 Param 1=4
    Index 0 Param 2=1
    Index 0 Param 3=1
    Index 0 Param 4=1
    Module.Index0Float0Write=true
    Module.Index0Float1Write=true
    Module.Index0Float2Write=true
    Module.Index0Float3Write=true
   )
   &DynamicMaterialParameters(
    Dynamic Parameter Output:ENiagara_AttributeSamplingApplyOutput=::NewEnumerator1
    Index 0 Param 4=&RandomRangeFloat(Minimum=0, Maximum=10, Evaluation Type=::NewEnumerator1)
    Module.Index0Float0Write=false
    Module.Index0Float1Write=false
    Module.Index0Float2Write=false
    Module.Index0Float3Write=true
   ) {module=DynamicMaterialParameters001}
   &DynamicMaterialParameters(
    Dynamic Parameter Output:ENiagara_AttributeSamplingApplyOutput=::NewEnumerator1
    Index 0 Param 1=&RandomRangeFloat(Minimum=0.8, Maximum=1.2)
    Index 0 Param 2=&RandomRangeFloat(Minimum=0.2, Maximum=0.4)
    Index 0 Param 3=&RandomRangeFloat(Minimum=0.4, Maximum=0.6)
    Index 0 Param 4=1
    Module.Index0Float0Write=true
    Module.Index0Float1Write=true
    Module.Index0Float2Write=true
    Module.Index0Float3Write=true
   ) {module=DynamicMaterialParameters003}
   &InitialMeshOrientation(
    Mesh Orientation Mode=::NewEnumerator7
    Rotation=&MergeHelper_VectorSwitch(
     0=<0 0 0>
     1=&MakeVector(
      X=&RandomRangeFloat(Minimum=-0.05, Maximum=0.05)
      Y=&RandomRangeFloat(Minimum=0, Maximum=1)
      Z=&RandomRangeFloat(Minimum=-0.05, Maximum=0.05)
     )
     2=&MakeVector(X=0, Y=&RandomRangeFloat(Minimum=0, Maximum=1), Z=0)
     3=?
     Index=$Particles.VisibilityTag
     Num=::NewEnumerator1
    )
    Use Rotation Vector=true
   )
   &AddVelocity(Velocity=&MergeHelper_VectorSwitch(
     0=<0 0 0>
     1=<0 0 0>
     2=&RandomRangeVector(Minimum=<0 -3 0>, Maximum=<0 -4 0>, Evaluation Type=::NewEnumerator1)
     3=<0 0 0>
     4=<0 0 0>
     5=<0 0 0>
     Index=$Particles.VisibilityTag
     Num=::NewEnumerator4
    ))
  ]
  particleUpdateScript [
   &ParticleState()
   &SpriteRotationRate(Rotation Rate=&MergeHelper_FloatSwitch(
     0=0
     1=0
     2=0
     3=0
     4=0
     5=&RandomRangeFloat(Minimum=-1, Maximum=1)
     Index=$Particles.VisibilityTag
     Num=::NewEnumerator4
    ))
   &ScaleColor(
    Scale Alpha=&Multiply_Float(
     A=&FloatFromCurve(FloatCurve=DataInterface() {propertyValues {Curve {keys [
          {interpMode=RCIM_Cubic, leaveTangentWeight=0.1333, tangentMode=RCTM_User, time=0, value=0}
          {arriveTangentWeight=0.2592, interpMode=RCIM_Cubic, leaveTangentWeight=0.1296, tangentMode=RCTM_User, time=0.4, value=1}
          {arriveTangentWeight=0.0667, interpMode=RCIM_Cubic, leaveTangentWeight=0.1333, tangentMode=RCTM_User, time=0.6, value=1}
          <at 1 0>
         ]}}})
     B=$System.SystemFade
    )
    ScaleA=true
    ScaleRGB=true
   )
   &UpdateMeshOrientation(
    Rotation Rate=&MergeHelper_FloatSwitch(
     0=0
     1=&Multiply_Float(
      A=0.02
      B=&UniformAOrBFloat(A=-1, Evaluation Type=::NewEnumerator0, Recalculate Random Each Loop=true, Override Seed=false)
     )
     2=0.1
     3=?
     Index=$Particles.VisibilityTag
     Num=::NewEnumerator1
    )
    Rotation Vector=<0 1 0>
   )
   &FN_Mesh_Attributes(
    Color=&MakeLinearColorFromVectorAndFloat(
     "Vector (RGB)"=&VectorFromCurve(VectorCurve=DataInterface() {propertyValues {
        XCurve {keys [<at 0 0 RCIM_Cubic>,<at 1 1 RCIM_Cubic>]}
        YCurve {keys [<at 0 1.0022 RCIM_Cubic>,<at 1 -0.4299 RCIM_Cubic>]}
        ZCurve {keys [<at 0 0 RCIM_Cubic>,<at 1 1 RCIM_Cubic>]}
       }})
     "Float (Alpha)"=&FloatFromCurve(FloatCurve=DataInterface() {propertyValues {Curve {keys [
          <at 0 0 RCIM_Cubic>
          {arriveTangent=-0.7516, interpMode=RCIM_Cubic, leaveTangent=-0.7516, tangentMode=RCTM_User, time=0.2513, value=0.5}
          <at 1 0 RCIM_Cubic>
         ]}}})
    )
    Create Color Attribute=true
    Module Output=::NewEnumerator1
   )
   &FN_Mesh_Attributes(
    Color=&MakeLinearColorFromVectorAndFloat(
     "Vector (RGB)"=<1 1 1>
     "Float (Alpha)"=&Multiply_Float(
      A=&FloatFromCurve(FloatCurve=DataInterface() {propertyValues {Curve {keys [
           <at 0 2.0136 RCIM_Cubic>
           {arriveTangent=-2.3524, interpMode=RCIM_Cubic, leaveTangent=-2.3524, time=0.3242, value=0.3996}
           {arriveTangent=2.391, interpMode=RCIM_Cubic, leaveTangent=2.391, time=0.6844, value=0.4035}
           <at 0.997 2.0082 RCIM_Cubic>
          ]}}})
      B=$System.SystemFade
     )
    )
    Create Color Attribute=true
    Create Scale Attribute=true
    Module Output=::NewEnumerator1
    Scale=&Multiply_VectorByFloat(
     Vector=<0.15 0.15 0.5>
     Float=&FloatFromCurve(FloatCurve=DataInterface() {propertyValues {Curve {keys [<at 0 0.6>,<at 1 1>]}}})
    )
   ) {module=FN_Mesh_Attributes001}
   &DynamicMaterialParameters(
    Dynamic Parameter Output:ENiagara_AttributeSamplingApplyOutput=::NewEnumerator1
    Index 0 Param 1=&FloatFromCurve(FloatCurve=DataInterface() {propertyValues {Curve {keys [<at 0 0>,<at 1 1>]}}})
    Index 0 Param 2=&RandomRangeFloat(Minimum=0, Maximum=10)
    Index 0 Param 3=1
    Index 0 Param 4=1
    Module.Index0Float0Write=true
    Module.Index0Float1Write=true
    Module.Index0Float2Write=true
    Module.Index0Float3Write=true
   ) {module=DynamicMaterialParameters002}
   &FN_Light_Attributes(
    Color=&MakeLinearColorFromVectorAndFloat(
     "Vector (RGB)"=&MakeVectorFromLinearColorRGB(Color=&VariantColorSelector(Variant Amount=1, Original Color=<rgba 0 5.0798 10 1>, Variant 1 Color=<rgba 2.48 0.4116 10 1>, Variant 2 Color=<rgba 1.4583 10 3.7329 1>))
     "Float (Alpha)"=&Multiply_Float(
      A=&FloatFromCurve(FloatCurve=DataInterface() {propertyValues {Curve {keys [<at 0 0 RCIM_Cubic>,<at 0.5 1 RCIM_Cubic>,<at 1 0 RCIM_Cubic>]}}})
      B=$System.SystemFade
     )
    )
    Create Color Attribute=true
    Create Exponent Attribute=true
    Create Radius Attribute=true
    Exponent=4
    Radius=35
    Scale Color=&FloatFromCurve(FloatCurve=DataInterface() {propertyValues {Curve {defaultValue=1, keys [<at 0 1 RCIM_Cubic>,<at 1 0.75 RCIM_Cubic>]}}})
   )
   &FN_Sprite_Attributes(
    Color=&MakeLinearColorFromVectorAndFloat(
     "Vector (RGB)"=&MakeVectorFromLinearColorRGB(Color=&VariantColorSelector(
       Variant Amount=1
       Original Color=<rgba 0.587 1.031 1.2 1>
       Variant 1 Color=<rgba 0.689 0.2073 1.5 1>
       Variant 2 Color=<rgba 0.2874 0.9647 0.4678 1>
      ))
     "Float (Alpha)"=&Multiply_Float(
      A=&FloatFromCurve(
       FloatCurve=DataInterface() {propertyValues {Curve {keys [
           <at 0 0 RCIM_Cubic>
           {arriveTangent=1.3333, interpMode=RCIM_Cubic, leaveTangent=1.3333, time=0.25, value=1}
           {interpMode=RCIM_Cubic, tangentMode=RCTM_User, time=0.75, value=1}
           <at 1 0>
          ]}}}
       Scale Curve=0.5
      )
      B=$System.SystemFade
     )
    )
    Create Color Attribute=true
    Create SpriteSize Attribute=true
    Module Output:ENiagara_AttributeSamplingApplyOutput=::NewEnumerator1
    SpriteSize=<40 40>
   )
   &FN_Sprite_Attributes(
    Color=&MakeLinearColorFromVectorAndFloat(
     "Vector (RGB)"=&MakeVectorFromLinearColorRGB(Color=&VariantColorSelector(Variant Amount=1, Original Color=<rgba 2.1266 12.2912 16 1>, Variant 1 Color=<rgba 7.4199 2.4 16 1>, Variant 2 Color=<rgba 1.6 16 5.4345 1>))
     "Float (Alpha)"=&FloatFromCurve(FloatCurve=DataInterface() {propertyValues {Curve {keys [<at 0 1>,<at 0.2 0.2>]}}})
    )
    Create Color Attribute=true
    Create SpriteSize Attribute=true
    Module Output:ENiagara_AttributeSamplingApplyOutput=::NewEnumerator1
    SpriteSize=&Vector2DFromFloat(Value=&Multiply_Float(
      A=&RandomRangeFloat(Minimum=3, Maximum=7)
      B=&FloatFromCurve(FloatCurve=DataInterface() {propertyValues {Curve {keys [
           <at -0.0044 0.2654>
           {arriveTangent=0.7558, interpMode=RCIM_Cubic, leaveTangent=0.7558, time=0.157, value=0.5261}
           {arriveTangent=-0.6241, interpMode=RCIM_Cubic, leaveTangent=-0.6241, time=0.7034, value=0.8003}
           <at 1 0>
          ]}}})
     ))
   ) {module=FN_Sprite_Attributes001}
   /Niagara/Modules/CascadeConversion/CascadeConversion_Orbit(
    Offset:Vector3f=&VectorFromCurve(VectorCurve=DataInterface() {propertyValues {
       XCurve {keys [{interpMode=RCIM_Cubic, tangentMode=RCTM_User, time=0, value=0}, <at 1 0 RCIM_Cubic>]}
       YCurve {keys [{arriveTangent=-10.1229, interpMode=RCIM_Cubic, leaveTangent=-10.1229, tangentMode=RCTM_User, time=0, value=0}, <at 1 -10 RCIM_Cubic>]}
       ZCurve {keys [{arriveTangent=10.0198, interpMode=RCIM_Cubic, leaveTangent=10.0198, tangentMode=RCTM_User, time=0, value=3}, <at 1 12 RCIM_Cubic>]}
      }})
    Rotation:Vector3f=&RandomRangeVector(Minimum=<0 0 0>, Maximum=<0 1 0>)
    Rotation Rate:Vector3f=&RandomRangeVector(Minimum=<0 0.05 0>, Maximum=<0 0.1 0>)
   )
   /Niagara/Modules/CascadeConversion/CascadeConversion_SolveOrbit()
   &SolveForcesAndVelocity(Clamp Velocity=false, Limit Acceleration=false)
  ]
  rendererClasses [
   &NiagaraSpriteRendererProperties
   &NiagaraMeshRendererProperties
   &NiagaraLightRendererProperties
  ]
  renderers [
   {rendererClass=&NiagaraSpriteRendererProperties}
   {rendererClass=&NiagaraMeshRendererProperties, rendererIndex=1}
   {rendererClass=&NiagaraMeshRendererProperties, rendererIndex=2}
   {rendererClass=&NiagaraLightRendererProperties, rendererIndex=3}
   {rendererClass=&NiagaraSpriteRendererProperties, rendererIndex=4}
   {rendererClass=&NiagaraSpriteRendererProperties, rendererIndex=5}
  ]
 }
 S2_MistParticle {
  emitterUpdateScript [
   &EmitterState(Inactive Response=::NewEnumerator1, Life Cycle Mode=::NewEnumerator1, Loop Duration=1, UseLoopDelay=true)
   &SpawnPerUnit(
    Module.Use Max Movement Threshold=true
    Module.Use Movement Tolerance=true
    Movement Tolerance=0.1
    Spawn Probability:NiagaraFloat=&MakeCustomFloatFromBool(
     Bool=&Boolean_OrOperation(
      Bool1=&SetBoolByIntComparison(A=&MakeIntFromFloat(Float=$User.Stage), B=0, Comparison Type=::NewEnumerator4)
      Bool2=&SetBoolByIntComparison(A=&MakeIntFromFloat(Float=$User.Stage), B=1, Comparison Type=::NewEnumerator4)
     )
     True Float=0
     False Float=1
    )
    Spawn Spacing=15
    Use Spawn Probability=true
   )
   &SpawnRate(
    Spawn Probability=&MakeCustomFloatFromBool(
     Bool=&Boolean_OrOperation(
      Bool1=&SetBoolByIntComparison(A=&MakeIntFromFloat(Float=$User.Stage), B=0, Comparison Type=::NewEnumerator4)
      Bool2=&Boolean_OrOperation(
       Bool1=&SetBoolByIntComparison(A=&MakeIntFromFloat(Float=$User.Stage), B=1, Comparison Type=::NewEnumerator4)
       Bool2=&SetBoolByIntComparison(A=&MakeIntFromFloat(Float=$User.Stage), B=2, Comparison Type=::NewEnumerator4)
      )
     )
     True Float=0
     False Float=1
    )
    SpawnGroup=1
    SpawnRate=&Multiply_Float(
     A=6
     B=&FloatFromCurve(
      FloatCurve=DataInterface() {propertyValues {Curve {keys [
          <at 0 1 RCIM_Cubic>
          {arriveTangent=-0.2543, interpMode=RCIM_Cubic, leaveTangent=-0.2543, time=1.2828, value=1.6362}
          {arriveTangent=-0.3703, interpMode=RCIM_Cubic, leaveTangent=-0.3703, time=2.4472, value=0.3776}
          {arriveTangent=-0.11, interpMode=RCIM_Cubic, leaveTangent=-0.11, time=3.6302, value=0.7671}
          {arriveTangent=0.2907, interpMode=RCIM_Cubic, leaveTangent=0.2907, time=5.034, value=0.0931}
          {arriveTangent=0.3058, interpMode=RCIM_Cubic, leaveTangent=0.3058, time=6.8459, value=1.702}
          <at 8 1>
         ]}}}
      CurveIndex=$Emitter.NormalizedLoopAge
     )
    )
    Use Spawn Probability=true
   ) {module=SpawnRate001}
   &SpawnPerUnit(
    Module.Use Max Movement Threshold=true
    Module.Use Movement Tolerance=true
    Movement Tolerance=0.1
    Spawn Group=1
    Spawn Probability:NiagaraFloat=&MakeCustomFloatFromBool(
     Bool=&Boolean_OrOperation(
      Bool1=&SetBoolByIntComparison(A=&MakeIntFromFloat(Float=$User.Stage), B=0, Comparison Type=::NewEnumerator4)
      Bool2=&Boolean_OrOperation(
       Bool1=&SetBoolByIntComparison(A=&MakeIntFromFloat(Float=$User.Stage), B=1, Comparison Type=::NewEnumerator4)
       Bool2=&SetBoolByIntComparison(A=&MakeIntFromFloat(Float=$User.Stage), B=2, Comparison Type=::NewEnumerator4)
      )
     )
     True Float=0
     False Float=1
    )
    Spawn Spacing=50
    Use Spawn Probability=true
   ) {module=SpawnPerUnit001}
  ]
  particleSpawnScript [
   /SetVariables(Particles.VisibilityTag=$Emitter.SpawnGroup) {isSetParameters=true}
   &InitializeParticle(
    Color=&VariantColorSelector(
     Variant Amount=1
     Original Color=<rgba 0.125 0.8195 1 1>
     Variant 1 Color=<rgba 0.369 0.0581 1 1>
     Variant 2 Color=<rgba 0.2211 0.9647 0.4191 1>
    )
    Lifetime=&MergeHelper_FloatSwitch(
     0=&RandomRangeFloat(Minimum=0.15, Maximum=0.2, Evaluation Type=::NewEnumerator1)
     1=&RandomRangeFloat(Minimum=0.5, Maximum=1, Evaluation Type=::NewEnumerator1)
     2=?
     3=?
     Index=$Particles.VisibilityTag
     Num=::NewEnumerator0
    )
    Position Offset=&MergeHelper_VectorSwitch(0=<-2 -15 1>, 1=<-5 -5 1>, 2=?, 3=?, Index=$Particles.VisibilityTag, Num=::NewEnumerator0)
    Position Offset Coordinate Space=::Local
    Sprite Size=<20 40>
    Sprite Size Mode=::NewEnumerator4
    UsePositionOffset=true
    Write Material Random=true
   )
   &InheritVelocity(
    Inherited Velocity Amount Scale=&MergeHelper_VectorSwitch(0=<0.1 0.1 0.1>, 1=<0.5 0.5 0.5>, 2=?, 3=?, Index=$Particles.VisibilityTag, Num=::NewEnumerator0)
    Inherited Velocity Speed Limit=0
    Speed Limit=false
    UseSpeedThreshold=false
   )
   &CameraOffset(Camera Offset Amount=&MergeHelper_FloatSwitch(0=-10, 1=0, 2=?, 3=?, Index=$Particles.VisibilityTag, Num=::NewEnumerator0))
   &ShapeLocation(
    MaskSpawnGroup=false
    Rotation Axis=<0 -1 0>
    Rotation Mode=::NewEnumerator1
    Sphere Distribution=::NewEnumerator1
    Sphere Radius=&MergeHelper_FloatSwitch(0=0, 1=5, 2=?, 3=?, Index=$Particles.VisibilityTag, Num=::NewEnumerator0)
    UseOwnerScale=true
   )
   &AddVelocity(
    Cone Angle=&MergeHelper_FloatSwitch(0=0, 1=45, 2=?, 3=?, Index=$Particles.VisibilityTag, Num=::NewEnumerator0)
    Cone Axis=<0 -1 0>
    Module.UseConeFalloff=false
    Module.UseVelDistribution=false
    Override Seed=false
    Velocity Mode=::NewEnumerator2
    Velocity Speed=&MergeHelper_FloatSwitch(
     0=1
     1=&RandomRangeFloat(Minimum=18, Maximum=36)
     2=?
     3=?
     Index=$Particles.VisibilityTag
     Num=::NewEnumerator0
    )
   )
  ]
  particleUpdateScript [
   &ParticleState()
   &SpriteRotationRate(Rotation Rate=&MergeHelper_FloatSwitch(
     0=0
     1=&RandomRangeFloat(Minimum=-1, Maximum=1)
     2=?
     3=?
     Index=$Particles.VisibilityTag
     Num=::NewEnumerator0
    ))
   &ScaleSpriteSize(
    Scale Factor=&Vector2DFromCurve(
     Vector2Curve=DataInterface() {propertyValues {
       XCurve {keys [<at 0 0.8031 RCIM_Cubic>, {arriveTangent=0.7722, interpMode=RCIM_Cubic, leaveTangent=0.7722, tangentMode=RCTM_User, time=1, value=1}]}
       YCurve {keys [<at 0 0.4789 RCIM_Cubic>, {interpMode=RCIM_Cubic, tangentMode=RCTM_User, time=1, value=1}]}
      }}
     ScaleCurve=<1 1>
    )
    Scale Sprite Size Mode=::NewEnumerator2
   )
   &Drag(
    Drag=&MergeHelper_FloatSwitch(0=4, 1=3, 2=?, 3=?, Index=$Particles.VisibilityTag, Num=::NewEnumerator0)
    Use Linear Drag=true
    Use Rotational Drag=true
   )
   /Niagara/Modules/Update/Forces/PointAttractionForce(
    Attraction Radius=500
    AttractionStrength:NiagaraFloat=&MergeHelper_FloatSwitch(0=20, 1=0, 2=?, 3=?, Index=$Particles.VisibilityTag, Num=::NewEnumerator0)
    AttractorPosition:Vector3f=&MakeVector(X=0, Y=0, Z=0)
    Falloff Exponent=0.1
    Module.Kill Within Radius=false
    Module.Use Falloff=true
   )
   &SolveForcesAndVelocity(Clamp Velocity=false, Limit Acceleration=false)
   &DynamicMaterialParameters(
    Dynamic Parameter Output:ENiagara_AttributeSamplingApplyOutput=::NewEnumerator1
    Index 0 Param 1=&FloatFromCurve(FloatCurve=DataInterface() {propertyValues {Curve {keys [{arriveTangent=2.8763, interpMode=RCIM_Cubic, leaveTangent=2.8763, tangentMode=RCTM_User, time=0, value=-2.0077}, <at 1 0.1406 RCIM_Cubic>]}}})
    Index 0 Param 2=&RandomRangeFloat(Minimum=0, Maximum=10)
    Index 0 Param 3=1
    Index 0 Param 4=1
    Module.Index0Float0Write=true
    Module.Index0Float1Write=true
    Module.Index0Float2Write=true
    Module.Index0Float3Write=true
   )
   &ScaleColor(
    Scale Alpha=&FloatFromCurve(FloatCurve=DataInterface() {propertyValues {Curve {keys [<at 0 0>,<at 0.3756 1>,<at 1 0>]}}})
    Scale RGB=&VectorFromCurve(VectorCurve=DataInterface() {propertyValues {XCurve {keys [<at 0 10>,<at 0.7 3>]}, YCurve {keys [<at 0 10>,<at 0.7 3>]}, ZCurve {keys [<at 0 10>,<at 0.7 3>]}}})
    ScaleA=true
    ScaleRGB=true
   )
   &FN_Sprite_Attributes(
    Color=&MakeLinearColorFromVectorAndFloat(
     "Vector (RGB)"=&MakeVectorFromLinearColorRGB(Color=&VariantColorSelector(Variant Amount=1, Original Color=<rgba 2.1266 12.2912 16 1>, Variant 1 Color=<rgba 7.4199 2.4 16 1>, Variant 2 Color=<rgba 1.6 16 5.4345 1>))
     "Float (Alpha)"=&FloatFromCurve(FloatCurve=DataInterface() {propertyValues {Curve {keys [<at 0 1>,<at 0.2 0.2>]}}})
    )
    Create Color Attribute=true
    Create SpriteSize Attribute=true
    Module Output:ENiagara_AttributeSamplingApplyOutput=::NewEnumerator1
    SpriteSize=&Vector2DFromFloat(Value=&Multiply_Float(
      A=&RandomRangeFloat(Minimum=3, Maximum=7)
      B=&FloatFromCurve(FloatCurve=DataInterface() {propertyValues {Curve {keys [
           <at -0.0044 0.2654>
           {arriveTangent=0.7558, interpMode=RCIM_Cubic, leaveTangent=0.7558, time=0.157, value=0.5261}
           {arriveTangent=-0.6241, interpMode=RCIM_Cubic, leaveTangent=-0.6241, time=0.7034, value=0.8003}
           <at 1 0>
          ]}}})
     ))
   )
  ]
  rendererClasses [&NiagaraSpriteRendererProperties]
  renderers [
   {rendererClass=&NiagaraSpriteRendererProperties}
   {rendererClass=&NiagaraSpriteRendererProperties, rendererIndex=1}
  ]
 }
 S3_Ghost {
  emitterUpdateScript [
   &EmitterState(
    Inactive Response=::NewEnumerator1
    Life Cycle Mode=::NewEnumerator1
    Loop Delay=&RandomRangeFloat(Minimum=0.5, Maximum=2, Evaluation Type=::NewEnumerator1)
    Loop Duration=1
    UseLoopDelay=true
   )
   &SpawnRate(SpawnRate=&Multiply_Float(A=2, B=$System.SpawnRateMp), Use Spawn Probability=false)
   &SpawnPerUnit(
    Max Movement Threshold=100
    Module.Use Max Movement Threshold=true
    Module.Use Movement Tolerance=true
    Movement Tolerance=0.1
    Spawn Spacing:NiagaraFloat=&Divide_Float(A=250, B=$System.SpawnRateMp)
    Use Spawn Probability=false
   )
   &SpawnRate(SpawnGroup=1, SpawnRate=&Multiply_Float(A=1, B=$System.SpawnRateMp), Use Spawn Probability=false) {module=SpawnRate001}
  ]
  particleSpawnScript [
   /SetVariables(Particles.VisibilityTag=$Emitter.SpawnGroup) {isSetParameters=true}
   &InitializeParticle(
    Color=&RandomRangeLinearColor(Minimum=<rgba 8 8 8 2>, Maximum=<rgba 4 4 4 4>)
    Lifetime=&MergeHelper_FloatSwitch(
     0=&RandomRangeFloat(Minimum=1, Maximum=2, Evaluation Type=::NewEnumerator1)
     1=&RandomRangeFloat(Minimum=0.5, Maximum=1)
     2=?
     3=?
     Index=$Particles.VisibilityTag
     Num=::NewEnumerator0
    )
    Position Offset=<-5 -3 1>
    Sprite Size=&RandomRangeVector2D(Minimum=<12 24>, Maximum=<16 35>, Evaluation Type=::NewEnumerator1)
    Sprite Size Mode=::NewEnumerator4
    UsePositionOffset=true
    Write Material Random=true
   )
   &AddVelocity(Velocity=&RandomRangeVector(Minimum=<-20 -40 -20>, Maximum=<20 -100 20>, Evaluation Type=::NewEnumerator1))
   &SubUVAnimation(
    End Frame Range Override=5
    Evaluation Type=::NewEnumerator0
    Output Alternate Renderer Binding=true
    "Sprite Renderer "=DataInterface() {propertyValues {SpriteRenderer=&self:S3_Ghost_0.NiagaraSpriteRendererProperties_1}}
    SubUV Animation Mode:ENiagara_SubUVLookupModeV2=::NewEnumerator2
    UseEndFrame=false
    UseStartFrame=false
   )
   &InheritVelocity(Inherited Velocity Amount Scale=<0.98 0.98 0.98>, Inherited Velocity Speed Limit=0, Speed Limit=false, UseSpeedThreshold=false)
  ]
  particleUpdateScript [
   &ParticleState()
   &ScaleSpriteSize(
    Scale Factor=&Vector2DFromFloat(Value=&FloatFromCurve(FloatCurve=DataInterface() {propertyValues {Curve {keys [{arriveTangent=3.006, interpMode=RCIM_Cubic, leaveTangent=3.006, tangentMode=RCTM_User, time=0, value=0}, <at 1 1 RCIM_Cubic>]}}}))
    Scale Sprite Size Mode=::NewEnumerator2
   )
   &PointAttractionForce(
    Attraction Radius=150
    AttractionStrength=30
    AttractorPosition:Vector3f=&RandomRangeVector(Minimum=<30 0 0>, Maximum=<60 0 0>, Evaluation Type=::NewEnumerator1)
    Falloff Exponent=0.1
    Module.Kill Within Radius=false
    Module.Use Falloff=true
   )
   &AccelerationForce(Acceleration=&RandomRangeVector(Minimum=<0 0 0>, Maximum=<0 0 50>))
   &Drag(Drag=2, Use Linear Drag=true, Use Rotational Drag=true)
   &SolveForcesAndVelocity(Clamp Velocity=false, Limit Acceleration=false)
   &ScaleColor(
    Scale Alpha=&FloatFromCurve(FloatCurve=DataInterface() {propertyValues {Curve {keys [<at 0 0 RCIM_Cubic>,<at 0.1044 1 RCIM_Cubic>,<at 1 0 RCIM_Cubic>]}}})
    Scale RGB=&MakeVectorFromLinearColorRGB(Color=&VariantColorSelector(
      Variant Amount=1
      Original Color=<rgba 0.152 0.607 1 1>
      Variant 1 Color=<rgba 0.369 0.0581 1 1>
      Variant 2 Color=<rgba 0.2874 0.9647 0.4678 1>
     ))
    ScaleA=true
    ScaleRGB=true
   )
   &DynamicMaterialParameters(
    Dynamic Parameter Output:ENiagara_AttributeSamplingApplyOutput=::NewEnumerator1
    Index 0 Param 1=1
    Index 0 Param 2=1
    Index 0 Param 3=&FloatFromCurve(FloatCurve=DataInterface() {propertyValues {Curve {keys [<at 0 0 RCIM_Cubic>,<at 0.3126 0.1502 RCIM_Cubic>,<at 1 0 RCIM_Cubic>]}}})
    Index 0 Param 4=&FloatFromCurve(FloatCurve=DataInterface() {propertyValues {Curve {keys [{arriveTangent=-29.0081, interpMode=RCIM_Cubic, leaveTangent=-29.0081, tangentMode=RCTM_User, time=0, value=-5.1116}, <at 1 -15 RCIM_Cubic>]}}})
    Module.Index0Float0Write=true
    Module.Index0Float1Write=true
    Module.Index0Float2Write=true
    Module.Index0Float3Write=true
   )
  ]
  rendererClasses [&NiagaraSpriteRendererProperties]
  renderers [
   {rendererClass=&NiagaraSpriteRendererProperties}
   {rendererClass=&NiagaraSpriteRendererProperties, rendererIndex=1}
  ]
 }
}
summary {
 systemName=NS_Backpack_GhostPortal_IdleFX
 emitters [
  {
   emitterName=DarkGradient_Packed
   rendererClasses [
    &NiagaraSpriteRendererProperties
    &NiagaraMeshRendererProperties
    &NiagaraLightRendererProperties
   ]
  }
  {emitterName=S3_Ghost, rendererClasses [&NiagaraSpriteRendererProperties]}
  {emitterName=S2_MistParticle, rendererClasses [&NiagaraSpriteRendererProperties]}
 ]
 userVariables [
  User.bIsFrontEnd {type=NiagaraBool}
  User.bIsFrontEndPreview {type=NiagaraBool}
  User.Stage {type=NiagaraFloat, defaultValue=0, description=""}
  User.Variant {type=NiagaraFloat, defaultValue=0, description=""}
 ]
}
