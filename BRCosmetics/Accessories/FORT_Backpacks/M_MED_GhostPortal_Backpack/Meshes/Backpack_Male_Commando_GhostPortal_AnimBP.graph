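# Animation Blueprint for the GhostPortal backpack on the male commando, adding secondary motion to its cloth and dangling chain bones.
# Above LOD 0, three world-space AnimDynamics cloth bones run over the baked pose; LOD 0 uses the plain baked pose.
# AnimDynamics chains and a single bone, limited by LOD threshold, then layer on top.
# Init caches the NvCloth interactor, but the skydiving stiffness branch has no incoming execution.
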AnimBlueprint /BRCosmetics/Accessories/FORT_Backpacks/M_MED_GhostPortal_Backpack/Meshes/Backpack_Male_Commando_GhostPortal_AnimBP
target_skeleton=&/Backpack_Male_Commando_GhostPortal_Skeleton
graphs [
 AnimGraph {nodes [
   Local Space Ref Pose()
   n1 = Anim Dynamics - Bone: BP_Cloth_01!(
    ComponentPose=Local To Component(LocalPose=Sequence Player 'GhostPortal_BakePose'() {sequence=&/GhostPortal_BakePose})
    Alpha=1
   ) -> (Pose) {
    angularDampingOverride=1
    angularSpringConstant=4000
    bAngularSpring=true
    bOverrideAngularDamping=true
    bUseGravityOverride=true
    gravityScale=0.35
    numSolverIterationsPreUpdate=3
    simulationSpace=World
    alphaBoolBlend {blendOption=Linear}
    boundBone {boneName=BP_Cloth_01}
    componentAppliedLinearAccClamp=<5000 5000 5000>
    componentLinearAccScale=<1 1 1>
    componentLinearVelScale=<1 1 1>
    gravityOverride {z=-980}
    physicsBodyDefinitions [{
      boundBone {boneName=BP_Cloth_01}
      boxExtents=<15 30 15>
      constraintSetup {angularLimitsMax=<10 0 20>, angularLimitsMin=<0 0 -20>, angularTarget=<0 0 0>}
      localJointOffset=<0 -2.5 0>
     }]
   }
   n2 = Anim Dynamics - Bone: BP_Cloth_02!(ComponentPose=@n1, Alpha=1) -> (Pose) {
    angularDampingOverride=1
    angularSpringConstant=4000
    bAngularSpring=true
    bOverrideAngularDamping=true
    bUseGravityOverride=true
    gravityScale=0.35
    numSolverIterationsPreUpdate=3
    simulationSpace=World
    alphaBoolBlend {blendOption=Linear}
    boundBone {boneName=BP_Cloth_02}
    componentAppliedLinearAccClamp=<5000 5000 5000>
    componentLinearAccScale=<1 1 1>
    componentLinearVelScale=<1 1 1>
    gravityOverride {z=-980}
    physicsBodyDefinitions [{
      boundBone {boneName=BP_Cloth_02}
      boxExtents=<15 25 20>
      constraintSetup {angularLimitsMax=<10 0 20>, angularLimitsMin=<0 0 -20>, angularTarget=<0 0 0>}
      localJointOffset=<0 -2 0>
     }]
   }
   n3 = Anim Dynamics - Bone: BP_Cloth_03!(ComponentPose=@n2, Alpha=1) -> (Pose) {
    angularDampingOverride=1
    angularSpringConstant=4000
    bAngularSpring=true
    bOverrideAngularDamping=true
    bUseGravityOverride=true
    gravityScale=0.35
    numSolverIterationsPreUpdate=3
    simulationSpace=World
    alphaBoolBlend {blendOption=Linear}
    boundBone {boneName=BP_Cloth_03}
    componentAppliedLinearAccClamp=<5000 5000 5000>
    componentLinearAccScale=<1 1 1>
    componentLinearVelScale=<1 1 1>
    gravityOverride {z=-980}
    physicsBodyDefinitions [{
      boundBone {boneName=BP_Cloth_03}
      boxExtents=<15 30 10>
      constraintSetup {angularLimitsMax=<10 0 20>, angularLimitsMin=<0 0 -20>, angularTarget=<0 0 0>}
      localJointOffset=<0 -3 0>
     }]
   }
   n4 = Component To Local(ComponentPose=@n3)
   n5 = Sequence Player 'GhostPortal_BakePose'() {sequence=&GhostPortal_BakePose}
   n6 = Get LODs() -> (LODs:bool)
   n7 = "Anim Dynamics - Chain: dyn_chain_01 -> dyn_chain_03"(
    ComponentPose=Local To Component(LocalPose=Blend Poses by bool(BlendPose_0=@n4, BlendPose_1=@n5, bActiveValue=@n6) {blendType=Linear})
    Alpha=0.85
    ExternalForce="0,0,0"
   ) {
    bChain=true
    bGravityOverrideInSimSpace=true
    lODThreshold=1
    numSolverIterationsPostUpdate=4
    numSolverIterationsPreUpdate=16
    alphaBoolBlend {blendOption=Linear}
    boundBone {boneName=dyn_chain_01}
    chainEnd {boneName=dyn_chain_03}
    componentAppliedLinearAccClamp=<1000 1000 1000>
    componentLinearAccScale=<1 1 1>
    componentLinearVelScale=<1 1 1>
    physicsBodyDefinitions [{
      sphereCollisionRadius=0
      boundBone {boneName=dyn_chain_01}
      boxExtents=<5 5 5>
      constraintSetup {angularTargetAxis=AxisZ, twistAxis=AxisZ, angularLimitsMax=<5 0.1 0>, angularLimitsMin=<-90 -0.1 0>, angularTarget=<0 0 0>}
      localJointOffset=<0 0 -0.5>
     }]
   }
   n8 = "Anim Dynamics - Chain: dyn_chain_04 -> dyn_chain_06"(ComponentPose=@n7, Alpha=0.75, ExternalForce="0,0,0") {
    bChain=true
    bGravityOverrideInSimSpace=true
    lODThreshold=1
    numSolverIterationsPostUpdate=4
    numSolverIterationsPreUpdate=16
    alphaBoolBlend {blendOption=Linear}
    boundBone {boneName=dyn_chain_04}
    chainEnd {boneName=dyn_chain_06}
    componentAppliedLinearAccClamp=<1000 1000 1000>
    componentLinearAccScale=<1 1 1>
    componentLinearVelScale=<1 1 1>
    physicsBodyDefinitions [{
      sphereCollisionRadius=0
      boundBone {boneName=dyn_chain_04}
      boxExtents=<5 5 5>
      constraintSetup {angularTargetAxis=AxisZ, twistAxis=AxisZ, angularLimitsMax=<5 0.1 0>, angularLimitsMin=<-90 -0.1 0>, angularTarget=<0 0 0>}
      localJointOffset=<0 0 -0.5>
     }]
   }
   n9 = "Anim Dynamics - Chain: dyn_chain_07 -> dyn_chain_09"(ComponentPose=@n8, Alpha=0.9, ExternalForce="0,0,0") {
    bChain=true
    bGravityOverrideInSimSpace=true
    lODThreshold=1
    numSolverIterationsPostUpdate=4
    numSolverIterationsPreUpdate=16
    alphaBoolBlend {blendOption=Linear}
    boundBone {boneName=dyn_chain_07}
    chainEnd {boneName=dyn_chain_09}
    componentAppliedLinearAccClamp=<1000 1000 1000>
    componentLinearAccScale=<1 1 1>
    componentLinearVelScale=<1 1 1>
    physicsBodyDefinitions [{
      sphereCollisionRadius=0
      boundBone {boneName=dyn_chain_07}
      boxExtents=<5 5 5>
      constraintSetup {angularTargetAxis=AxisZ, twistAxis=AxisZ, angularLimitsMax=<5 0.1 0>, angularLimitsMin=<-45 -0.1 0>, angularTarget=<0 0 0>}
      localJointOffset=<0 0 -0.5>
     }]
   }
   n10 = Anim Dynamics - Bone: dyn_chain_10(ComponentPose=@n9, Alpha=0.6, ExternalForce="0,0,0") {
    bGravityOverrideInSimSpace=true
    lODThreshold=1
    numSolverIterationsPostUpdate=4
    numSolverIterationsPreUpdate=16
    alphaBoolBlend {blendOption=Linear}
    boundBone {boneName=dyn_chain_10}
    componentAppliedLinearAccClamp=<1000 1000 1000>
    componentLinearAccScale=<1 1 1>
    componentLinearVelScale=<1 1 1>
    physicsBodyDefinitions [{
      sphereCollisionRadius=0
      boundBone {boneName=dyn_chain_10}
      boxExtents=<5 5 5>
      constraintSetup {angularTargetAxis=AxisZ, twistAxis=AxisZ, angularLimitsMax=<5 0.1 0>, angularLimitsMin=<-90 -0.1 0>, angularTarget=<0 0 0>}
      localJointOffset=<0 0 -0.5>
     }]
   }
   Output Pose(Result=Component To Local(ComponentPose=@n10)) {layerGroup=None}
  ]}
 EventGraph {nodes [
   Event BlueprintUpdateAnimation()
   > Set LODs(LODs:bool="integer > integer"(A=Get CurrentLOD(), B=0) {is=title K2Node_CallFunction}) -> (Output_Get:bool)
   Event BlueprintInitializeAnimation()
   > Set random_stream(random_stream:Random Stream Structure=MakeRandomStream!(InitialSeed:int=0) -> (ReturnValue:Random Stream Structure)) -> (Output_Get:Random Stream Structure)
   > n1 = Cast To ClothingSimulationInteractorNv!(
    execute
    Object=GetClothingSimulationInteractor!(self:Skeletal Mesh Component Object Reference=GetPartSkeletalMeshComponent(), ClothingAssetName:Name=None) -> (ReturnValue:Clothing Simulation Interactor Object Reference)
   ) -> (then, CastFailed, AsClothing Simulation Interactor Nv:Clothing Simulation Interactor Nv Object Reference) {is=title K2Node_DynamicCast}
   > Set clothing_sim_interactor(clothing_sim_interactor:Clothing Simulation Interactor Nv Object Reference=@n1) -> (Output_Get:Clothing Simulation Interactor Nv Object Reference)
   n2 = Get bIsSkyDiving()
   n3 = Branch(Condition=@n2)
   n4 = Get clothing_sim_interactor() -> (clothing_sim_interactor:Clothing Simulation Interactor Nv Object Reference)
   > SetAnimDriveSpringStiffness!(execute, self:Clothing Simulation Interactor Object Reference=@n4, InStiffness:float=20) -> (then)
   > SetAnimDriveDamperStiffness!(execute, self:Clothing Simulation Interactor Nv Object Reference=@n4, InStiffness:float=10) -> (then)
   > Do Once(Reset=@n5.then, Start Closed=true)
   > ForceClothNextUpdateTeleportAndReset(self:Skeletal Mesh Component Object Reference=GetOwnerSkeletalMeshComponent())
   SetAnimDriveSpringStiffness(execute=@n3.else, self=@n4, InStiffness=0)
   > n5 = SetAnimDriveDamperStiffness(self=@n4, InStiffness=0)
   OR Boolean(A=@n2, B=Get bCanPlayCustomAnimations() -> (bCanPlayCustomAnimations:bool))
  ]}
]
variables [
 clothing_sim_interactor {type=ClothingSimulationInteractorNv, default=null}
 LODs {type=boolean}
 random_stream {type=RandomStream, default {initialSeed=0, seed=0}}
]
