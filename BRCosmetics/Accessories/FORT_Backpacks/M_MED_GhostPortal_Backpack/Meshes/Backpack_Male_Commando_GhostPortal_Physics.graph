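# Physics asset for the Ghost Portal backpack cosmetic mesh (male commando, medium), supplying its collision shapes.
# A single body on the backpack_root bone approximates the pack's volume with three angled capsules of similar size.
# It has no constraints, so it defines no articulated or jointed parts.
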PhysicsAsset /BRCosmetics/Accessories/FORT_Backpacks/M_MED_GhostPortal_Backpack/Meshes/Backpack_Male_Commando_GhostPortal_Physics
bodies {backpack_root [
  {shapeType=Capsule, length=7.25, radius=5.75, center=<4.2853 -2.9627 0.0763>, rotation=<rot 44.8139 -1.8899 -4.5818>}
  {shapeType=Capsule, length=5.75, radius=6.5, center=<-7.2271 -1.1925 10.728>, rotation=<rot 4.5788 4.1956 84.9828>}
  {shapeType=Capsule, length=5.75, radius=7, center=<-7.5203 -0.7228 -11.7442>, rotation=<rot 4.5788 4.1967 84.9828>}
 ]}
constraints []
